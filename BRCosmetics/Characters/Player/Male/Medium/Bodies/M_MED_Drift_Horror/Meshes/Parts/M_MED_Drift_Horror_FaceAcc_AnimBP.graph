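# Animation Blueprint for the face accessory part of the male Drift Horror player cosmetic, adding secondary motion to its hair bones.
# It copies the pose from the attached parent mesh, then chains seven single-body AnimDynamics nodes over dyn_hair_01 to 07.
# Each uses damped angular springs and angular limits, simulating root-relative or, for 02–04, bone-relative to root.
# The EventGraph has no logic and there are no variables.
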AnimBlueprint /BRCosmetics/Characters/Player/Male/Medium/Bodies/M_MED_Drift_Horror/Meshes/Parts/M_MED_Drift_Horror_FaceAcc_AnimBP
target_skeleton=&/M_MED_Drift_Horror_FaceAcc_Skeleton
graphs [
 AnimGraph {nodes [
   n1 = Anim Dynamics - Bone: dyn_hair_01(ComponentPose=Local To Component(LocalPose=Copy Pose From Mesh() {bUseAttachedParent=true}), Alpha=1) {
    angularDampingOverride=0.9999
    angularSpringConstant=100
    bAngularSpring=true
    bGravityOverrideInSimSpace=true
    bOverrideAngularDamping=true
    simulationSpace=RootRelative
    alphaBoolBlend {blendOption=Linear}
    boundBone {boneName=dyn_hair_01}
    componentAppliedLinearAccClamp=<3500 3500 3500>
    componentLinearAccScale=<1 1 1>
    componentLinearVelScale=<1 1 1>
    physicsBodyDefinitions [{
      sphereCollisionRadius=0
      boundBone {boneName=dyn_hair_01}
      boxExtents=<15 8 8>
      constraintSetup {angularLimitsMax=<5 5 15>, angularLimitsMin=<-5 -35 -15>, angularTarget=<0 0 0>}
      localJointOffset=<2 0 0>
     }]
    relativeSpaceBone {boneName=root}
   }
   n2 = Anim Dynamics - Bone: dyn_hair_02(ComponentPose=@n1, Alpha=1) {
    angularDampingOverride=0.9999
    angularSpringConstant=100
    bAngularSpring=true
    bGravityOverrideInSimSpace=true
    bOverrideAngularDamping=true
    simulationSpace=BoneRelative
    alphaBoolBlend {blendOption=Linear}
    boundBone {boneName=dyn_hair_02}
    componentAppliedLinearAccClamp=<3500 3500 3500>
    componentLinearAccScale=<1 1 1>
    componentLinearVelScale=<1 1 1>
    physicsBodyDefinitions [{
      sphereCollisionRadius=0
      boundBone {boneName=dyn_hair_02}
      boxExtents=<15 10 10>
      constraintSetup {angularLimitsMax=<5 15 15>, angularLimitsMin=<-5 -45 -15>, angularTarget=<0 0 0>}
      localJointOffset=<2 0 0>
     }]
    relativeSpaceBone {boneName=root}
   }
   n3 = Anim Dynamics - Bone: dyn_hair_03(ComponentPose=@n2, Alpha=1, ExternalForce="0,0,0") {
    angularDampingOverride=0.9999
    angularSpringConstant=100
    bAngularSpring=true
    bGravityOverrideInSimSpace=true
    bOverrideAngularDamping=true
    simulationSpace=BoneRelative
    alphaBoolBlend {blendOption=Linear}
    boundBone {boneName=dyn_hair_03}
    componentAppliedLinearAccClamp=<3500 3500 3500>
    componentLinearAccScale=<1 1 1>
    componentLinearVelScale=<1 1 1>
    physicsBodyDefinitions [{
      sphereCollisionRadius=0
      boundBone {boneName=dyn_hair_03}
      boxExtents=<15 10 10>
      constraintSetup {angularLimitsMax=<5 15 15>, angularLimitsMin=<-5 -45 -15>, angularTarget=<0 0 0>}
      localJointOffset=<2 0 0>
     }]
    relativeSpaceBone {boneName=root}
   }
   n4 = Anim Dynamics - Bone: dyn_hair_04(ComponentPose=@n3, Alpha=1, ExternalForce="0,0,0") {
    angularDampingOverride=0.9999
    angularSpringConstant=300
    bAngularSpring=true
    bGravityOverrideInSimSpace=true
    bOverrideAngularDamping=true
    simulationSpace=BoneRelative
    alphaBoolBlend {blendOption=Linear}
    boundBone {boneName=dyn_hair_04}
    componentAppliedLinearAccClamp=<3500 3500 3500>
    componentLinearAccScale=<1.1 1.1 1.1>
    componentLinearVelScale=<1.1 1.1 1.1>
    physicsBodyDefinitions [{
      sphereCollisionRadius=0
      boundBone {boneName=dyn_hair_04}
      boxExtents=<15 10 10>
      constraintSetup {angularLimitsMax=<1 30 15>, angularLimitsMin=<-1 -45 -25>, angularTarget=<0 0 0>}
      localJointOffset=<3 0 0>
     }]
    relativeSpaceBone {boneName=root}
   }
   n5 = Anim Dynamics - Bone: dyn_hair_05(ComponentPose=@n4, Alpha=1, ExternalForce="0,0,0") {
    angularDampingOverride=0.9999
    angularSpringConstant=100
    bAngularSpring=true
    bGravityOverrideInSimSpace=true
    bOverrideAngularDamping=true
    simulationSpace=RootRelative
    alphaBoolBlend {blendOption=Linear}
    boundBone {boneName=dyn_hair_05}
    componentAppliedLinearAccClamp=<3500 3500 3500>
    componentLinearAccScale=<1 1 1>
    componentLinearVelScale=<1 1 1>
    physicsBodyDefinitions [{
      sphereCollisionRadius=0
      boundBone {boneName=dyn_hair_05}
      boxExtents=<20 10 10>
      constraintSetup {angularLimitsMax=<1 5 25>, angularLimitsMin=<-1 -45 -15>, angularTarget=<0 0 0>}
      localJointOffset=<3 0 0>
     }]
    relativeSpaceBone {boneName=root}
   }
   n6 = Anim Dynamics - Bone: dyn_hair_06(ComponentPose=@n5, Alpha=1, ExternalForce="0,0,0") {
    angularDampingOverride=0.9999
    angularSpringConstant=100
    bAngularSpring=true
    bGravityOverrideInSimSpace=true
    bOverrideAngularDamping=true
    simulationSpace=RootRelative
    alphaBoolBlend {blendOption=Linear}
    boundBone {boneName=dyn_hair_06}
    componentAppliedLinearAccClamp=<3500 3500 3500>
    componentLinearAccScale=<0.8 0.8 0.8>
    componentLinearVelScale=<0.8 0.8 0.8>
    physicsBodyDefinitions [{
      sphereCollisionRadius=0
      boundBone {boneName=dyn_hair_06}
      boxExtents=<20 10 10>
      constraintSetup {angularLimitsMax=<1 5 15>, angularLimitsMin=<-1 -30 -15>, angularTarget=<0 0 0>}
      localJointOffset=<2 0 0>
     }]
    relativeSpaceBone {boneName=root}
   }
   n7 = Anim Dynamics - Bone: dyn_hair_07(ComponentPose=@n6, Alpha=1, ExternalForce="0,0,0") {
    angularDampingOverride=0.9999
    angularSpringConstant=100
    bAngularSpring=true
    bGravityOverrideInSimSpace=true
    bOverrideAngularDamping=true
    simulationSpace=RootRelative
    alphaBoolBlend {blendOption=Linear}
    boundBone {boneName=dyn_hair_07}
    componentAppliedLinearAccClamp=<3500 3500 3500>
    componentLinearAccScale=<0.8 0.8 0.8>
    componentLinearVelScale=<0.8 0.8 0.8>
    physicsBodyDefinitions [{
      sphereCollisionRadius=0
      boundBone {boneName=dyn_hair_07}
      boxExtents=<20 10 10>
      constraintSetup {angularLimitsMax=<1 5 15>, angularLimitsMin=<-1 -30 -15>, angularTarget=<0 0 0>}
      localJointOffset=<2 0 0>
     }]
    relativeSpaceBone {boneName=root}
   }
   Output Pose(Result=Component To Local(ComponentPose=@n7)) {layerGroup=None}
  ]}
 EventGraph {nodes [Event BlueprintUpdateAnimation()]}
]
variables []
